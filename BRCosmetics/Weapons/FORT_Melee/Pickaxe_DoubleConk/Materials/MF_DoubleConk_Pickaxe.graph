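# Material function for the DoubleConk pickaxe that layers an animated, glowing grid effect over incoming material attributes.
# Local-position sphere masks, roughened by grid noise, define transition regions; grid lines, time-based fading, bump-offset diffuse and a faked depth gradient supply the look.
# A sweeping band uses pre-skinned position, passed through a vertex interpolator for pixel shading.
# Named reroutes feed blends that override BaseColor and Emissive and fade Metallic and Specular to zero.
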MaterialFunction /BRCosmetics/Weapons/FORT_Melee/Pickaxe_DoubleConk/Materials/MF_DoubleConk_Pickaxe
graph {nodes [
  n1 = MaterialFunctionCall(Float3=NamedRerouteUsage()) {materialFunction=&BreakOutFloat3Components}
  n2 = Add(A=AppendVector(A=@n1.R, B=@n1.R), B=Constant2Vector() {g=-2.15})
  Add(A=@n2, B=Multiply(B=ConstantBiasScale() {bias=-0.5, scale=2}) {constA=0.02})
  Constant()
  n3 = FunctionInput() {inputType=FunctionInput_MaterialAttributes}
  n4 = GetMaterialAttributes(None=@n3) {attributeGetTypes [
    #1
    #2
    #3
    #4
   ]}
  n5 = NamedRerouteUsage()
  n6 = NamedRerouteUsage()
  n7 = VectorParameter() {parameterName=Emissive1, defaultValue=<rgba 0.0857 0.1378 0.2083 1>}
  n8 = VectorParameter() {parameterName=Emissive2, defaultValue=<rgba 0 1.4 3 1>}
  n9 = NamedRerouteUsage()
  n10 = LinearInterpolate(A=@n6, B=Add(A=@n6, B=LinearInterpolate(A=@n7.RGB, B=@n8.RGB, Alpha=@n9))) {constAlpha=0.85}
  n11 = Multiply(A=@n10) {constB=0.5}
  n12 = Multiply(
   A=@n10
   B=LinearInterpolate(
    A=ScalarParameter() {defaultValue=0.85, parameterName=Emissive_Final_Min}
    B=ScalarParameter() {defaultValue=3, parameterName=Emissive_Final_Max}
    Alpha=NamedRerouteUsage()
   )
  )
  n13 = NamedRerouteUsage()
  n14 = LinearInterpolate(
   A=NamedRerouteDeclaration(None=@n4.EmissiveColor) {name=EmissiveIn, nodeColor=<rgba 1 0.2578 0 1>}
   B=Add(A=LinearInterpolate(A=@n11, B=@n12, Alpha=@n13), B=NamedRerouteUsage())
   Alpha=NamedRerouteUsage()
  )
  n15 = Multiply(A=NamedRerouteUsage(), B=NamedRerouteUsage())
  n16 = SetMaterialAttributes!(
   MaterialAttributes=@n3
   BaseColor=LinearInterpolate(A=@n4.BaseColor, B=Multiply(A=@n5) {constB=0.25}, Alpha=NamedRerouteUsage())
   EmissiveColor=Multiply(A=Add(A=@n14, B=@n15), B=NamedRerouteUsage())
   Metallic=LinearInterpolate(A=@n4.Metallic, Alpha=NamedRerouteUsage()) {constB=0}
   Specular=LinearInterpolate(A=@n4.Specular, Alpha=NamedRerouteUsage()) {constB=0}
  ) {attributeSetTypes [
    #2
    #1
    #3
    #4
   ]}
  FunctionOutput(None=@n16)
  n17 = MaterialFunctionCall(Power=Constant() {r=0.5}) {materialFunction=&Fresnel_Function}
  Multiply(A=@n17.Result)
  n18 = Divide(A=LocalPosition().XYZ) {constB=26.5}
  n19 = MaterialFunctionCall(Float3=@n18) {materialFunction=&BreakOutFloat3Components}
  n20 = NamedRerouteDeclaration(None=MaterialFunctionCall(X=@n19.R, Y=@n19.R) {materialFunction=&MakeFloat2}.Result) {name=UV_RB, nodeColor=<rgba 0 0.3672 1 1>}
  NamedRerouteDeclaration(None=TextureSample(UVs=Panner(Coordinate=@n20) {speedY=0.3}) {texture=/BRCosmetics/Weapons/FORT_Melee/Pickaxe_DoubleConk/Textures/T_GridNoise}.R) {name=Grid_RandomPanning, nodeColor=<rgba 0.8984 0 1 1>}
  n21 = Multiply(A=NamedRerouteUsage(), B=ScalarParameter() {defaultValue=0.2, parameterName=Transition_GridNoiseStrength})
  NamedRerouteDeclaration(None=SphereMask(
    A=Add(A=@n19.B, B=@n21)
    B=Constant() {r=2.4}
    Radius=ScalarParameter() {defaultValue=1.15, parameterName=Transition_Radius}
    Hardness=ScalarParameter() {defaultValue=0.3, parameterName=Transition_Hardness}
   ) {attenuationRadius=1.35, hardnessPercent=50}) {name=TransitionMask, nodeColor=<rgba 0 1 0.5781 1>}
  n22 = Multiply(A=NamedRerouteUsage()) {constB=16}
  NamedRerouteDeclaration(None=Multiply(
    A=TextureSample(UVs=@n22) {texture=/BRCosmetics/Weapons/FORT_Melee/Pickaxe_DoubleConk/Textures/T_GridDebug}.RGB
    B=VectorParameter() {parameterName=GridLineColor, defaultValue=<rgba 0 5.3961 20 1>}.RGB
   )) {name=GridLine, nodeColor=<rgba 1 0.1875 0 1>}
  n23 = TextureSampleParameter2D(UVs=BumpOffset(Coordinate=TextureCoordinate(), Height=ScalarParameter() {defaultValue=-0.1, parameterName=BaseColor_BumpOffsetHeight}) {heightRatio=0.02}) {
   parameterName=Diffuse
   samplerSource=SSM_Wrap_WorldGroupSettings
   texture=/BRCosmetics/Weapons/FORT_Melee/Pickaxe_DoubleConk/Textures/T_DoubleConk_Pickaxe_D
  }
  n24 = NamedRerouteUsage()
  n25 = Saturate(None=SphereMask(A=Divide(A=@n24) {constB=100}, B=Constant() {r=0.46}) {attenuationRadius=0.02, hardnessPercent=0})
  NamedRerouteDeclaration(None=LinearInterpolate(A=@n23.RGB, B=NamedRerouteUsage(), Alpha=@n25) {constB=0}) {name=Diffuse_Out, nodeColor=<rgba 0.3125 1 0 1>}
  NamedRerouteDeclaration(None=LinearInterpolate(B=NamedRerouteUsage(), Alpha=@n25) {constB=0}) {name=Emissive_BumpOffset, nodeColor=<rgba 0 0.6484 1 1>}
  n26 = Multiply(A=NamedRerouteUsage(), B=ScalarParameter() {defaultValue=0.2, parameterName=Transition_GridNoiseStrength})
  NamedRerouteDeclaration(None=SphereMask(
    A=Add(A=@n19.B, B=@n26)
    B=Constant() {r=2.6}
    Radius=ScalarParameter() {defaultValue=1.15, parameterName=Transition_Radius}
    Hardness=Constant() {r=0.9}
   ) {attenuationRadius=1.35, hardnessPercent=50}) {name=HarshMask, nodeColor=<rgba 0.5469 1 0 1>}
  NamedRerouteDeclaration(None=SphereMask(A=@n19.B, B=Constant() {r=2.4}) {attenuationRadius=1.25, hardnessPercent=0}) {name=VerticalMask, nodeColor=<rgba 0 0.5078 1 1>}
  NamedRerouteDeclaration(None=LinearInterpolate(Alpha=@n17.Result) {constA=0.5}) {name=Fresnel, nodeColor=<rgba 1 0 0.7031 1>}
  n27 = LocalPosition() {localOrigin=InstancePreSkinning}
  n28 = MaterialFunctionCall(Float3=ShaderStageSwitch(PixelShader=VertexInterpolator(VS=@n27.XYZ).PS, VertexShader=@n27.XYZ)) {materialFunction=&BreakOutFloat3Components}
  NamedRerouteDeclaration(None=@n28.B) {name=PreskinnedZ, nodeColor=<rgba 0 1 0.0156 1>}
  NamedRerouteDeclaration(None=@n4.BaseColor) {name=BaseColorIn, nodeColor=<rgba 0 1 0.625 1>}
  n29 = Time()
  n30 = ScalarParameter() {defaultValue=0.5, parameterName=GridFadingSpeed}
  n31 = Add(A=NamedRerouteUsage(), B=Frac(None=Multiply(A=@n29, B=@n30)))
  n32 = Multiply(A=ConstantBiasScale(None=Sine(None=@n31)), B=ScalarParameter() {defaultValue=0.5, group=STGridFX, parameterName=GridFadingAmount})
  n33 = NamedRerouteUsage()
  NamedRerouteDeclaration(None=Saturate(None=Multiply(A=@n32, B=@n33))) {name=Grid_Random_Fading, nodeColor=<rgba 1 0.6563 0 1>}
  NamedRerouteDeclaration(None=TextureSample(UVs=@n20) {texture=&T_GridNoise}.R) {name=Grid_Random, nodeColor=<rgba 0.8984 0 1 1>}
  NamedRerouteDeclaration(None=@n18) {name=LocalPosition, nodeColor=<rgba 0.2422 0 1 1>}
  n34 = Abs(None=@n28.R)
  n35 = AppendVector(A=@n28.R, B=Add(A=Power(Base=@n34) {constExponent=1.2}, B=@n28.B))
  n36 = ScalarParameter() {defaultValue=100, group=STGridFX, parameterName=BandSpreadingLocalPositionDivider}
  n37 = ScalarParameter() {parameterName=BandCenterPositionX}
  n38 = ScalarParameter() {defaultValue=1.3, parameterName=BandCenterPositionZ}
  n39 = OneMinus(None=SphereMask(A=Divide(A=@n35, B=@n36) {constB=100}, B=AppendVector(A=@n37, B=@n38)) {attenuationRadius=2, hardnessPercent=0})
  n40 = Multiply(A=NamedRerouteUsage(), B=ScalarParameter() {defaultValue=0.1, group=STGridFX, parameterName=BandSpreadingNoiseAmount}) {constB=0.1}
  n41 = Time()
  n42 = ScalarParameter() {defaultValue=-0.24, group=STGridFX, parameterName=BandSpreadingSpeed}
  n43 = Power(Base=Frac(None=Multiply(A=@n41, B=@n42)))
  n44 = ScalarParameter() {defaultValue=1.3, group=STGridFX, parameterName=BandLocationMAX}
  n45 = Power(Base=SphereMask(
    A=Add(A=@n39, B=@n40)
    B=Multiply(A=@n43, B=@n44) {constB=1.1}
    Radius=ScalarParameter() {defaultValue=0.05, group=STGridFX, parameterName=BandWidth}
   ) {attenuationRadius=0.1, hardnessPercent=0.5}) {constExponent=0.5}
  n46 = ScalarParameter() {defaultValue=2, group=STGridFX, parameterName=BandEmissiveMult}
  NamedRerouteDeclaration(None=Multiply(A=Multiply(A=@n45, B=@n46), B=VectorParameter() {group=STGridFX, parameterName=BandColor, defaultValue=<rgba 0 0.5 1 1>}.RGB)) {name=Band, nodeColor=<rgba 0.5703 1 0 1>}
  n47 = MaterialFunctionCall(Float2=@n2) {materialFunction=&BreakOutFloat2Components}
  n48 = If(A=@n47.G, "A > B"=@n47.G, "A == B"=@n47.G, "A < B"=Multiply(A=@n47.G) {constB=0.1})
  n49 = NamedRerouteDeclaration(None=Add(A=Abs(None=@n47.R), B=Abs(None=@n48))) {name=FakeDepth, nodeColor=<rgba 0.0781 1 0 1>}
  n50 = Constant()
  n51 = Constant() {r=1}
  n52 = Power(Base=@n49) {constExponent=0.5}
  n53 = Power(Base=NamedRerouteUsage()) {constExponent=1}
  n54 = Constant() {r=1}
  n55 = Add(
   A=Multiply(A=@n49) {constB=1.5}
   B=Multiply(A=MaterialFunctionCall(A=@n50, B=@n51, Transition Phase=@n52, Height Texture=@n53, Contrast=@n54) {materialFunction=&HeightLerp}.Results) {constB=0.1}
  )
  NamedRerouteDeclaration(None=Power(Base=Saturate(None=@n55), Exp=ScalarParameter() {defaultValue=1.2, parameterName=GridDepthPow}) {constExponent=1.5}) {name=GridDepth, nodeColor=<rgba 0 0.7187 1 1>}
  NamedRerouteUsage()
  ScalarParameter() {defaultValue=1, parameterName=Transition_Hardness}
 ]}
