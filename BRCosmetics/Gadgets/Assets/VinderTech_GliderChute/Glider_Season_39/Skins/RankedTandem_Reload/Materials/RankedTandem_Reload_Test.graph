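# Test character-style material for the Season 39 RankedTandem Reload glider skin, built on material attributes for skinned, morph and cloth meshes.
# The shared 2023 character parent PreFX function feeds its result through the DoubleConk pickaxe's material function, borrowed from weapon cosmetics, then into PostFX.
# PreFX also feeds the PostFX auxiliary inputs directly.
# Shading model comes from the expression, with the character subsurface profile assigned.
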Material /BRCosmetics/Gadgets/Assets/VinderTech_GliderChute/Glider_Season_39/Skins/RankedTandem_Reload/Materials/RankedTandem_Reload_Test
graph {
 properties {
  bUseMaterialAttributes=true
  bUsedWithClothing=true
  bUsedWithMeshDeformer=true
  bUsedWithMorphTargets=true
  bUsedWithSkeletalMesh=true
  pixelDepthOffsetMode=PDOM_Legacy
  shadingModel=MSM_FromMaterialExpression
  subsurfaceProfile=/Game/Packages/Fortress_SharedMaterials/SubSurfaceProfile/Character_SubsurfaceProfile
  usedShadingModels=Default Lit | Subsurface | Clear Coat | Subsurface Profile | Cloth
 }
 nodes [
  n1 = MaterialFunctionCall() {materialFunction=&MF_CharacterParent_2023_PreFX}
  MaterialFunctionCall(
   Material=MaterialFunctionCall!(In=@n1.Result) {materialFunction=/BRCosmetics/Weapons/FORT_Melee/Pickaxe_DoubleConk/Materials/MF_DoubleConk_Pickaxe}.Result
   M=@n1.M
   Pre Skinned Local Position=@n1.M
   Distance Blend=@n1.M
   Fresnel=@n1.M
  ) {materialFunction=&MF_CharacterParent_2023_PostFX}
 ]
}
